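# Animation Blueprint for the DairyString Wade backpack cosmetic, adding secondary physics sway to its banana and straw bones.
# The input pose gets a small additive offset on the pack bone in component space, then passes through two chained Anim Dynamics nodes.
# Both use angular springs, damping overrides, reduced gravity and angular limits; bananas simulate root-relative, while the straw simulates in actor space with amplified component motion.
# The event graph contains no working logic.
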AnimBlueprint /BRCosmetics/Accessories/FORT_Backpacks/Backpack_M_MED_DairyString_Wade/Meshes/M_MED_DairyString_Wade_Pack_AnimBP
target_skeleton=&/M_MED_DairyString_Wade_Pack_Skeleton
graphs [
 AnimGraph {nodes [
   n1 = Input Pose()
   n2 = Anim Dynamics - Bone: dyn_bananas_c(
    ComponentPose="Transform (Modify) Bone - Bone: pack"(ComponentPose=Local To Component(LocalPose=@n1), Translation="0,-0.6,0", Rotation="0,0,0") {translationMode=BMM_Additive, translationSpace=BCS_BoneSpace, boneToModify {boneName=pack}, translation {y=-0.6}}
    Alpha=1
    ExternalForce="0,0,0"
   ) {
    angularDampingOverride=1
    angularSpringConstant=13000
    bAngularSpring=true
    bOverrideAngularDamping=true
    bUseGravityOverride=true
    gravityScale=0.35
    simulationSpace=RootRelative
    boundBone {boneName=dyn_bananas_c}
    componentAppliedLinearAccClamp=<5000 5000 5000>
    componentLinearAccScale=<0.3 0.3 0.3>
    componentLinearVelScale=<0.3 0.3 0.2>
    gravityOverride {z=-980}
    physicsBodyDefinitions [{
      boundBone {boneName=dyn_bananas_c}
      boxExtents=<35 39 35>
      constraintSetup {angularLimitsMax=<0 28 5>, angularLimitsMin=<0 -25 -5>, angularTarget=<1 0 0>}
      localJointOffset=<3.2 0 0>
     }]
    simSpaceSettings {maxAngularAcceleration=100, maxAngularVelocity=10, simSpaceAngularAlpha=0.1}
   }
   n3 = Anim Dynamics - Bone: dyn_straw_c(ComponentPose=@n2, Alpha=1, ExternalForce="0,0,0") {
    angularDampingOverride=1
    angularSpringConstant=19300
    bAngularSpring=true
    bOverrideAngularDamping=true
    bUseGravityOverride=true
    gravityScale=0.35
    simulationSpace=Actor
    boundBone {boneName=dyn_straw_c}
    componentAppliedLinearAccClamp {y=5000, z=5000}
    componentLinearAccScale=<1.3 2.6 1.3>
    componentLinearVelScale=<1.5 2.6 1.5>
    gravityOverride {z=-980}
    physicsBodyDefinitions [{
      boundBone {boneName=dyn_straw_c}
      boxExtents=<35 35.5 38>
      constraintSetup {angularLimitsMax=<0 24.5 14>, angularLimitsMin=<0 -7.5 -3>, angularTarget=<1 0 0>}
      localJointOffset=<2.4 0 0>
     }]
   }
   Output Pose(Result=Component To Local(ComponentPose=@n3)) {layerGroup=None}
  ]}
 EventGraph {nodes [
   TryGetPawnOwner()
   Event BlueprintUpdateAnimation()
  ]}
]
variables []
